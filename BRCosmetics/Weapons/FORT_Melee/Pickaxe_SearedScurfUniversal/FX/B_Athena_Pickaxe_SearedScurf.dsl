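; Weapon Blueprint for the SearedScurf cosmetic pickaxe, driving its idle, swing and impact effects.
; On equip it checks for the SearedScurf body to set idle VFX parameters, and adds a blank-material axe mesh on the right hand as the VFX anchor.
; Swing events derive a swing index from side and montage section, which angles impact effects.
; A Shadow Stance gameplay cue flag gates swing handling, and hiding the weapon deactivates idle VFX.
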(event EventOnWeaponVisibilityChanged
  (params (bVisible :type Boolean) (bSetForLocalControllerOnly :type Boolean))
  (|CallParentFunction bVisible bSetForLocalControllerOnly)
  (if (not bVisible)
    (Components|Activation|Deactivate (CallFunction|GetIdleVFXComponent))))

(event GameplayCue|GameplayCue.Perk.Ninja.ShadowStance.Active
  (params (EventType :type EGameplayCueEvent%20Enum :default "OnActive") (Parameters :type Gameplay%20Cue%20Parameters%20Structure%20%28by%20ref%29))
  (switch Utilities|FlowControl|Switch|SwitchonEGameplayCueEvent EventType
    (:OnActive
      _)
    (:WhileActive
      _)
    (:Executed
      _)
    (:Removed
      (Variables|Default|SetShadowStance false)))
  (Variables|Default|SetShadowStance true))

(event FortWeapon|EventOnEquippedAndAttached
  (bind _returnvalue (CallFunction|GetIdleVFXComponent))
  (bind _returnvalue_2 (Game|GetInstigator))
  (bind _asfort_player_pawn (Utilities|Casting|CastToFortPlayerPawn _returnvalue_2))
  (|CallParentFunction)
  (Networking|CanExecuteCosmeticEvents
    (:True
      (Utilities|FlowControl|Sequence
        (:then_0
          (Utilities|FlowControl|Delay 0.2))
        (:then_1
          (Utilities|FlowControl|Delay 1.0))
        (:then_2
          (Utilities|FlowControl|Delay 2.0)))
      (if (not (== (Variables|/Script/Engine.SkeletalMeshComponent|GetSkeletalMeshAsset (Customization|GetSkeletalMeshforPartType _asfort_player_pawn "Body" :!class /Script/FortniteGame.FortPlayerPawn)) "/BRCosmetics/Characters/Player/Male/Medium/Bodies/M_MED_SearedScurf/Meshes/M_MED_SearedScurf.M_MED_SearedScurf"))
        (Effects|Components|ParticleSystem|SetIntParameter (CallFunction|GetIdleVFXComponent) "FXint32" 2))
      (Utilities|FlowControl|DoOnce false
        (:Completed
          (Variables|Default|SetIsBuildInVersion true)
          (bind _returnvalue_3 (AddComponent|NODE_AddSkeletalMeshComponent-0 :self (Utilities|Casting|CastToFortPlayerPawn (Game|GetInstigator)) :bManualAttachment true))
          (Transformation|AttachComponentToComponent _returnvalue_3 (Variables|/Script/Engine.Character|GetMesh _asfort_player_pawn) "Hand_R" "SnapToTarget" "SnapToTarget" "SnapToTarget" true)
          (Components|SkeletalMesh|SetSkeletalMeshAsset _returnvalue_3 "/BRCosmetics/Weapons/FORT_Melee/Pickaxe_SearedScurfUniversal/Meshes/SearedScurfUniversal_Axe.SearedScurfUniversal_Axe")
          (Transformation|SetRelativeTransform _returnvalue_3 (Math|Transform|MakeTransform "-5.973945,1.943224,-0.826237" "90.000000,180.000000,-180.000000") false false)
          (Rendering|Material|SetMaterial _returnvalue_3 0 "/Game/Gadgets/Assets/VinderTech_GliderChute/Glider_Hippie/Materials/M_Blank.M_Blank")
          (Transformation|AttachComponentToComponent (CallFunction|GetIdleVFXComponent) _returnvalue_3 "Melee_TwoHanded_Handle" "SnapToTarget" "SnapToTarget" "SnapToTarget" true)
          (Effects|Components|ParticleSystem|SetBoolParameter _returnvalue "IsBuildInVersion" true)
          (Effects|Components|ParticleSystem|SetIntParameter _returnvalue "FXint32" 1))))))

(event EventSwingLeft
  (bind _returnvalue (Game|GetInstigator))
  (bind _returnvalue_2 (CallFunction|GetSwingVFXComponent))
  (bind _asfort_player_pawn (Utilities|Casting|CastToFortPlayerPawn _returnvalue))
  (bind _mesh (Variables|/Script/Engine.Character|GetMesh _asfort_player_pawn))
  (bind _returnvalue_3 (Components|SkeletalMesh|GetAnimInstance _mesh))
  (bind _returnvalue_4 (Animation|Montage|GetCurrentActiveMontage _returnvalue_3))
  (bind _returnvalue_5 (Animation|Montage|MontageGetCurrentSection _returnvalue_3 _returnvalue_4))
  (if (not (Variables|Default|GetShadowStance))
    (|CallParentFunction)
    (Utilities|FlowControl|Sequence
      (:then_0
        (Variables|Default|SetSwingIndex 0))
      (:then_1
        (if (== _returnvalue_5 "Swing3")
          (Variables|Default|SetSwingIndex 2)))
      (:then_2
        (if (== _returnvalue_5 "Swing4")
          (Variables|Default|SetSwingIndex 3)))
      (:then_3
        (Utilities|IsValid _returnvalue_2
          (:Is%20Valid
            (Transformation|AttachComponentToComponent _returnvalue_2 (stale FortWeapon|EventOnEquippedAndAttached _returnvalue_3) "Melee_TwoHanded_Handle" "SnapToTarget" "SnapToTarget" "SnapToTarget" true)))))))

(event EventSwingRight
  (bind _returnvalue (Game|GetInstigator))
  (bind _returnvalue_2 (CallFunction|GetSwingVFXComponent))
  (bind _asfort_player_pawn (Utilities|Casting|CastToFortPlayerPawn _returnvalue))
  (bind _mesh (Variables|/Script/Engine.Character|GetMesh _asfort_player_pawn))
  (bind _returnvalue_3 (Components|SkeletalMesh|GetAnimInstance _mesh))
  (bind _returnvalue_4 (Animation|Montage|GetCurrentActiveMontage _returnvalue_3))
  (bind _returnvalue_5 (Animation|Montage|MontageGetCurrentSection _returnvalue_3 _returnvalue_4))
  (if (not (Variables|Default|GetShadowStance))
    (|CallParentFunction)
    (Utilities|FlowControl|Sequence
      (:then_0
        (Variables|Default|SetSwingIndex 1))
      (:then_1
        (if (== _returnvalue_5 "Swing3")
          (Variables|Default|SetSwingIndex 2)))
      (:then_2
        (if (== _returnvalue_5 "Swing4")
          (Variables|Default|SetSwingIndex 3)))
      (:then_3
        (Utilities|IsValid _returnvalue_2
          (:Is%20Valid
            (Transformation|AttachComponentToComponent _returnvalue_2 (stale FortWeapon|EventOnEquippedAndAttached _returnvalue_3) "Melee_TwoHanded_Handle" "SnapToTarget" "SnapToTarget" "SnapToTarget" true)))))))

(event FortWeapon|EventOnPlayImpactFX
  (params (HitResult :type Hit%20Result%20Structure%20%28by%20ref%29) (ImpactPhysicalSurface :type EPhysicalSurface%20Enum :default "SurfaceType_Default") (SpawnedPSC :type FXSystem%20Component%20Object%20Reference))
  (bind (_bblockinghit _binitialoverlap _time _distance _location _impactpoint _normal) (Collision|BreakHitResult HitResult))
  (|CallParentFunction HitResult ImpactPhysicalSurface SpawnedPSC)
  (Transformation|SetWorldRotation SpawnedPSC (Math|Rotator|CombineRotators (Math|Rotator|MakeRotfromX _normal) (Utilities|Select :Index (Variables|Default|GetSwingIndex) :Option%200 "0.000000,-60.000000,20.000000" :Option%201 "0.000000,60.000000,10.000000" :Option%202 "0.000000,60.000000,-20.000000" :Option%203 "0.000000,-60.000000,-20.000000")) false false))

(fn SetSwingAudioParam
  (params (InInt :type Integer))
  (bind _returnvalue (AudioParameter|GetAudioParameterComponent))
  (Utilities|IsValid _returnvalue
    (:Is%20Valid
      (Audio|Parameter|SetIntegerParameter _returnvalue "SwingIdx" InInt :!class /Script/AudioExtensions.AudioParameterControllerInterface))))

(fn ConstructionScript
  (|CallParentFunction))
